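# AMB_Hatchery_Loop_Cue is a looping hatchery ambience Sound Cue in the Season 17 Battle Pass BushRanger quest sounds.
# A single Wave Player loops the AMB_Hatchery_Loop_v2 wave directly as the root, with no mixing or randomization.
# The cue uses the shared tile-ambient attenuation and Athena outdoor ambient sound class, at roughly half volume.
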SoundCue /BattlepassS17/Sounds/S17Quests/BushRanger/AMB_Hatchery_Loop_Cue
root=WavePlayer_5
nodes [{id=WavePlayer_5, properties {bLooping=true, soundWaveAssetPtr=/BattlepassS17/Sounds/S17Quests/BushRanger/Waves/AMB_Hatchery_Loop_v2}}]
properties {
 attenuationSettings=/Game/Sounds/Attenuation/Fort_TileAmbient_Att
 maxDistance=10000
 soundClassObject=&Fort_Ambient_Outdoors_Athena
 volumeMultiplier=0.501
 preEffectBusSends [{}]
}
