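# Sound cue played when a hatchery is constructed in the Season 17 BushRanger quest.
# A mixer layers a reward-screen chest upgrade chime with a slightly quieter-to-louder balance favoring the wood construction layer.
# That layer fires after a short fixed delay and picks randomly among three equally weighted wood "construction done" waves.
# It uses the non-critical in-game UI sound class, with concurrency limited to one instance, stopping the oldest.
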SoundCue /BattlepassS17/Sounds/S17Quests/BushRanger/HatcheryConstructed_Cue
root=Mixer_0
nodes [
 {id=Delay_0, children [Random_0], properties {delayMax=0.1, delayMin=0.1}}
 {id=Mixer_0, children [WavePlayer_1,Delay_0], properties {inputVolume [0.4,0.7]}}
 {id=Random_0, children [WavePlayer_0,WavePlayer_2,WavePlayer_3], properties {bShouldExcludeFromBranchCulling=true, weights [1,1,1]}}
 {id=WavePlayer_0, properties {soundWaveAssetPtr=/BattlepassS17/Sounds/S17Quests/BushRanger/Waves/Fort_Construction_Done_Wood_01}}
 {id=WavePlayer_1, properties {soundWaveAssetPtr=/Game/Sounds/Fort_UI_Sounds/Rewards/Fort_RewardScreen_TeamScore_ChestLootUpgrade}}
 {id=WavePlayer_2, properties {soundWaveAssetPtr=/BattlepassS17/Sounds/S17Quests/BushRanger/Waves/Fort_Construction_Done_Wood_02}}
 {id=WavePlayer_3, properties {soundWaveAssetPtr=/BattlepassS17/Sounds/S17Quests/BushRanger/Waves/Fort_Construction_Done_Wood_03}}
]
properties {
 bOverrideConcurrency=true
 maxDistance=2097152
 soundClassObject=&Fort_UI_InGame_NonCritical
 volumeMultiplier=1
 concurrencyOverrides {maxCount=1, resolutionRule=StopOldest}
}
